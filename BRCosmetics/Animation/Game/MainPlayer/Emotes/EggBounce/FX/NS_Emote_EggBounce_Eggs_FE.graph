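# Niagara system for the EggBounce emote, apparently its front-end variant, that shows up to two colored eggs which wobble, fly and burst.
# Eggs draws both eggs from one particle using system-level colors, masks and offsets, and plays a hatch sound before expiring.
# Glitter and smoke emitters read Eggs, spawning within age windows before Egg Burst Time.
# Debris and sprite-burst emitters vary their looks per spawn group.
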NiagaraSystem /BRCosmetics/Animation/Game/MainPlayer/Emotes/EggBounce/FX/NS_Emote_EggBounce_Eggs_FE
dependencies {
 usedDataInterfaces [
  &NiagaraDataInterfaceAudioPlayer
  &NiagaraDataInterfaceCurve
  &NiagaraDataInterfaceVectorCurve
  &NiagaraDataInterfaceParticleRead
  &NiagaraDataInterfaceVector2DCurve
  &NiagaraDataInterfaceColorCurve
  &NiagaraDataInterfaceArrayColor
  &NiagaraDataInterfaceArrayInt32
  &NiagaraDataInterfaceArrayFloat3
 ]
 usedDynamicInputs [
  &MakeFloatFromVector2d
  &SelectIntFromArray
  &SelectVectorFromArray
  /Niagara/DynamicInputs/TypeConversions/MakeCustomQuaternionFromBool
  &SetBoolByIntComparison
  &MakeQuatFromAxisAngle
  &VectorLength
  &Multiply_Vector
 ]
 usedModules [
  &self:SystemSpawnScript.SystemScriptSource.SystemScriptGraph.NiagaraNodeAssignment_4.SetVariables_#1
  &self:SystemSpawnScript.SystemScriptSource.SystemScriptGraph.NiagaraNodeAssignment_0.SetVariables_#2
  &self:SystemSpawnScript.SystemScriptSource.SystemScriptGraph.NiagaraNodeAssignment_1.SetVariables_#3
  &self:SystemSpawnScript.SystemScriptSource.SystemScriptGraph.NiagaraNodeAssignment_2.SetVariables_#4
  &self:SystemSpawnScript.SystemScriptSource.SystemScriptGraph.NiagaraNodeAssignment_3.SetVariables_#5
  &self:SystemSpawnScript.SystemScriptSource.SystemScriptGraph.NiagaraNodeAssignment_7.SetVariables_#6
  &self:SystemSpawnScript.SystemScriptSource.SystemScriptGraph.NiagaraNodeAssignment_8.SetVariables_#7
  &self:SystemSpawnScript.SystemScriptSource.SystemScriptGraph.NiagaraNodeAssignment_9.SetVariables_#8
  &self:SystemSpawnScript.SystemScriptSource.SystemScriptGraph.NiagaraNodeAssignment_19.SetVariables_#9
  &self:SystemSpawnScript.SystemScriptSource.SystemScriptGraph.NiagaraNodeAssignment_15.SetVariables_#10
  &self:SystemSpawnScript.SystemScriptSource.SystemScriptGraph.NiagaraNodeAssignment_10.SetVariables_#11
  &self:SystemSpawnScript.SystemScriptSource.SystemScriptGraph.NiagaraNodeAssignment_20.SetVariables_#12
  &self:SystemSpawnScript.SystemScriptSource.SystemScriptGraph.NiagaraNodeAssignment_16.SetVariables_#13
  &self:SystemSpawnScript.SystemScriptSource.SystemScriptGraph.NiagaraNodeAssignment_13.SetVariables_#14
  &self:SystemSpawnScript.SystemScriptSource.SystemScriptGraph.NiagaraNodeAssignment_14.SetVariables_#15
  &self:SystemSpawnScript.SystemScriptSource.SystemScriptGraph.NiagaraNodeAssignment_6.SetVariables_#16
  &SystemState
  &self:SystemSpawnScript.SystemScriptSource.SystemScriptGraph.NiagaraNodeAssignment_5.SetVariables_#17
 ]
}
emitters {
 Debris_Sprite {
  simTarget=GPUComputeSim
  emitterUpdateScript [
   &EmitterState(Life Cycle Mode=::NewEnumerator1, Loop Behavior=::NewEnumerator1, Loop Delay=0.05, Loop Duration=0.5, UseLoopDelay=false)
   &SpawnBurst_Instantaneous(
    Spawn Count=&RandomRangeInt(Minimum=5, Maximum=7)
    Use Spawn Probability=false
    UseLoopCountLimit=false
   )
   &SpawnBurst_Instantaneous(
    Spawn Count=&RandomRangeInt(Minimum=3, Maximum=4)
    Spawn Group=1
    Use Spawn Probability=false
    UseLoopCountLimit=false
   ) {module=SpawnBurst_Instantaneous001}
   &SpawnBurst_Instantaneous(Spawn Count=&RandomRangeInt(Minimum=7, Maximum=10), Spawn Group=2, Use Spawn Probability=false) {module=SpawnBurst_Instantaneous002}
  ]
  particleSpawnScript [
   /SetVariables(Particles.VisibilityTag=&SpawnGroup()) {isSetParameters=true}
   &InitializeParticle(
    Color=&MaskLinearColorBySpawnGroup(
     Masked Linear Color=&MakeCustomLinearColorFromBool(
      Bool=$User.bIsFrontEndPreview
      True Color=&RandomRangeLinearColor(Minimum=<rgba 0.3177 0.1972 0.1605 0.35>, Maximum=<rgba 0.3021 0.2182 0.1731 0.5>)
      False Color=&RandomRangeLinearColor(Minimum=<rgba 0.375 0.2073 0.1562 0.35>, Maximum=<rgba 0.375 0.2328 0.1562 0.5>)
     )
     Passthrough Linear Color=&MaskLinearColorBySpawnGroup(Masked Linear Color=<rgba 1 1 1 1>, Passthrough Linear Color=<rgba 0.4167 0.4167 0.4167 0.4957>, Spawn Group Mask=1)
    )
    Lifetime=&MaskFloatBySpawnGroup(
     Masked Float=&RandomRangeFloat(Minimum=0.3, Maximum=1)
     Passthrough Float=&MaskFloatBySpawnGroup(
      Masked Float=&MaskFloatBySpawnGroup(Masked Float=0.2, Passthrough Float=0.4)
      Passthrough Float=0.3
      Spawn Group Mask=1
     )
    )
    Mass Mode=::NewEnumerator1
    Position Offset=&MaskVectorBySpawnGroup(
     Masked Vector=<0 -15 0>
     Passthrough Vector=&MaskVectorBySpawnGroup(Masked Vector=<0 -15 -7>, Passthrough Vector=<0 -9 -8>, Spawn Group Mask=1)
    )
    Sprite Rotation Mode=::NewEnumerator2
    Sprite Size=&MaskVector2dBySpawnGroup(
     Masked Vector2D=&RandomRangeVector2D(Minimum=<5 5>, Maximum=<10 10>)
     Passthrough Vector2D=&MaskVector2dBySpawnGroup(
      Masked Vector2D=<1.5 1.5>
      Passthrough Vector2D=&RandomRangeVector2D(
       Minimum=&Multiply_Vector2D_ByFloat(Float=0.6, Vector2D=<40 62>)
       Maximum=&Multiply_Vector2D_ByFloat(Float=0.65, Vector2D=<20 50>)
      )
      Spawn Group Mask=1
     )
     Spawn Group Mask=0
    )
    Sprite Size Mode=::NewEnumerator4
    UsePositionOffset=true
    Write Material Random=true
   )
   &ShapeLocation(
    Handle Radius=1
    Large Radius=&MaskFloatBySpawnGroup(
     Masked Float=3
     Passthrough Float=&MaskFloatBySpawnGroup(Masked Float=0, Passthrough Float=4, Spawn Group Mask=1)
    )
    MaskSpawnGroup=false
    Shape Primitive=::NewEnumerator3
    Torus Distribution Mode:ENiagaraTorusDistributionMode=::NewEnumerator1
    U Position=&MaskFloatBySpawnGroup(
     Masked Float=&RandomRangeFloat(Minimum=0, Maximum=1)
     Passthrough Float=&MaskFloatBySpawnGroup(
      Masked Float=&RandomRangeFloat(Minimum=0, Maximum=1)
      Passthrough Float=&ReturnNormalizedExecIndex()
      Spawn Group Mask=1
     )
    )
    UseOwnerScale=true
   )
   &InitialMeshOrientation(
    Mesh Axis To Orient=<0 1 0>
    Mesh Orientation Mode=::NewEnumerator2
    Orientation Vector=$Particles.ShapeLocation.ShapeTangent
    Use Rotation Vector=true
   )
   &AddVelocity(
    Module.Use Falloff Distance=false
    Origin Offset=&MaskVectorBySpawnGroup(
     Masked Vector=&RandomRangeVector(Minimum=<0 0 -1>, Maximum=<0 0 -5>)
     Passthrough Vector=&MaskVectorBySpawnGroup(
      Masked Vector=&RandomRangeVector(Minimum=<-10 -10 -8>, Maximum=<10 10 -50>)
      Passthrough Vector=<0 -9 -10>
      Spawn Group Mask=1
     )
    )
    Override Seed=false
    Velocity Mode=::NewEnumerator1
    Velocity Origin=&AddVectorToPosition(
     Position=$Engine.Emitter.SimulationPosition
     Vector=&MaskVectorBySpawnGroup(
      Masked Vector=<0 -15 0>
      Passthrough Vector=&MaskVectorBySpawnGroup(Masked Vector=<0 -15 0>, Passthrough Vector=<0 0 0>, Spawn Group Mask=1)
     )
    )
    Velocity Speed=&MaskFloatBySpawnGroup(
     Masked Float=&RandomRangeFloat(Minimum=150, Maximum=300)
     Passthrough Float=&MaskFloatBySpawnGroup(Masked Float=&RandomRangeFloat(Minimum=150, Maximum=200), Passthrough Float=75, Spawn Group Mask=1)
    )
   )
   &CameraOffset(Camera Offset Amount=&MaskFloatBySpawnGroup(
     Masked Float=0
     Passthrough Float=&MaskFloatBySpawnGroup(Masked Float=0, Passthrough Float=1, Spawn Group Mask=1)
    ))
  ]
  particleUpdateScript [
   &ParticleState()
   &GravityForce(Gravity=&MaskVectorBySpawnGroup(
     Masked Vector=<0 0 -980>
     Passthrough Vector=&MaskVectorBySpawnGroup(Masked Vector=<0 0 0>, Passthrough Vector=<0 0 0>, Spawn Group Mask=1)
    ))
   &Drag(
    Drag=&MaskFloatBySpawnGroup(
     Masked Float=1
     Passthrough Float=&MaskFloatBySpawnGroup(
      Masked Float=&FloatFromCurve(
       FloatCurve=DataInterface() {propertyValues {Curve {keys [
           {interpMode=RCIM_Cubic, tangentMode=RCTM_User, time=0, value=0}
           {arriveTangent=3.2375, interpMode=RCIM_Cubic, leaveTangent=3.2375, tangentMode=RCTM_User, time=1, value=1}
          ]}}}
       Scale Curve=102.9192
      )
      Passthrough Float=0
      Spawn Group Mask=1
     )
    )
    Use Linear Drag=true
    Use Rotational Drag=true
   )
   &AccelerationForce(Acceleration=&MaskVectorBySpawnGroup(
     Masked Vector=<0 100 0>
     Passthrough Vector=&MaskVectorBySpawnGroup(Masked Vector=<0 100 0>, Passthrough Vector=<0 0 0>, Spawn Group Mask=1)
    ))
   &SolveForcesAndVelocity(Clamp Velocity=false, Limit Acceleration=false)
   &ScaleSpriteSizeBySpeed(
    Max Scale Factor=&MaskVector2dBySpawnGroup(
     Masked Vector2D=<1 1>
     Passthrough Vector2D=&MaskVector2dBySpawnGroup(Masked Vector2D=<0.4 20>, Passthrough Vector2D=<1 1>, Spawn Group Mask=1)
     Spawn Group Mask=0
    )
    Min Scale Factor=<1 0.75>
    Velocity Threshold=500
   )
   &ScaleSpriteSize(Scale Factor=&MaskVector2dBySpawnGroup(
     Masked Vector2D=&Vector2DFromFloat(Value=&FloatFromCurve(FloatCurve=DataInterface() {propertyValues {Curve {keys [<at 0.5 1 RCIM_Cubic>, {arriveTangent=-3.7562, interpMode=RCIM_Cubic, leaveTangent=-3.7562, tangentMode=RCTM_User, time=1, value=0}]}}}))
     Passthrough Vector2D=&MaskVector2dBySpawnGroup(
      Masked Vector2D=&Vector2DFromFloat(Value=&FloatFromCurve(FloatCurve=DataInterface() {propertyValues {Curve {keys [<at 0.5 1 RCIM_Cubic>, {arriveTangent=-3.7562, interpMode=RCIM_Cubic, leaveTangent=-3.7562, tangentMode=RCTM_User, time=1, value=0}]}}}))
      Passthrough Vector2D=&Vector2DFromCurve(
       Vector2Curve=DataInterface() {propertyValues {XCurve {keys [<at 0 0>,<at 1 1>]}, YCurve {keys [<at 0 0>,<at 0.2 0.775>,<at 1 0>]}}}
       ScaleCurve=&Multiply_Vector2D_ByFloat(Float=&RandomRangeFloat(Minimum=0.8, Maximum=1), Vector2D=<1 1>)
      )
      Spawn Group Mask=1
     )
     Spawn Group Mask=0
    ))
   &ScaleColor(
    Scale Alpha=&MaskFloatBySpawnGroup(
     Masked Float=1
     Passthrough Float=&MaskFloatBySpawnGroup(
      Masked Float=&FloatFromCurve(FloatCurve=DataInterface() {propertyValues {Curve {keys [{interpMode=RCIM_Cubic, tangentMode=RCTM_User, time=0, value=1}, {interpMode=RCIM_Cubic, tangentMode=RCTM_User, time=1, value=0}]}}})
      Passthrough Float=&FloatFromCurve(
       FloatCurve=DataInterface() {propertyValues {Curve {keys [
           {arriveTangent=-4.0245, interpMode=RCIM_Cubic, leaveTangent=-4.0245, tangentMode=RCTM_User, time=0, value=1.5}
           {arriveTangent=-1.5, interpMode=RCIM_Cubic, leaveTangent=-1.5, time=0.2, value=0.8}
           <at 1 0>
          ]}}}
       Scale Curve=&RandomRangeFloat(Minimum=0.05, Maximum=0.7)
      )
      Spawn Group Mask=1
     )
    )
    Scale RGB=&MaskVectorBySpawnGroup(
     Masked Vector=<1 1 1>
     Passthrough Vector=&MaskVectorBySpawnGroup(Masked Vector=<1 1 1>, Passthrough Vector=&VectorFromFloat(Value=0.4), Spawn Group Mask=1)
    )
    ScaleA=true
    ScaleRGB=true
   )
   /SetVariables(Particles.Position=&Lerp_Vector(
     A=Linked() {linkedVariable {name=Particles.Position, type=NiagaraPosition}}
     B=&AddVectorToPosition(
      Position=$Particles.Position
      Vector=&MaskVectorBySpawnGroup(
       Masked Vector=<0 -20 0>
       Passthrough Vector=&MaskVectorBySpawnGroup(Masked Vector=<0 -20 0>, Passthrough Vector=<0 -27 0>, Spawn Group Mask=1)
      )
     )
     Alpha=&FloatFromCurve(FloatCurve=DataInterface() {propertyValues {Curve {keys [{tangentMode=RCTM_User, time=0, value=0}, {tangentMode=RCTM_User, time=1, value=0.4}]}}})
    )) {isSetParameters=true}
  ]
  rendererClasses [&NiagaraSpriteRendererProperties]
  renderers [
   {rendererClass=&NiagaraSpriteRendererProperties}
   {rendererClass=&NiagaraSpriteRendererProperties, rendererIndex=1}
   {rendererClass=&NiagaraSpriteRendererProperties, rendererIndex=2}
  ]
 }
 EggBurst_Glitter {
  emitterUpdateScript [
   &EmitterState()
   &SpawnBurst_Instantaneous(
    Spawn Count=&MakeCustomIntFromBool(Bool=$System.bIsMoving, True INT=100, False INT=0)
    Spawn Time=&Add_Float(A=$System.Egg Burst Time, B=-0.25)
    Use Spawn Probability=false
    UseLoopCountLimit=false
   ) {enabled=false}
   &SpawnParticlesFromOtherEmitter(
    Attribute Reader=DataInterface() {propertyValues {EmitterBinding {bindingMode=Other, emitterName=Eggs}}}
    Calculate Spawn Rate Per Particle=false
    Maximum Age Spawn Threshold:NiagaraFloat=&Add_Float(A=$System.Egg Burst Time, B=-0.24)
    Minimum Age Spawn Threshold:NiagaraFloat=&Add_Float(A=$System.Egg Burst Time, B=-0.34)
    Spawn Rate=1000
    Use Maximum Age=true
    Use Minimum Age=true
   )
  ]
  particleSpawnScript [
   &FN_SetSpawnGroup(Number of Spawn Groups:NiagaraInt32=&Add_Integer(A=$System.EggVisibility, B=1))
   &InitializeParticle(
    Color=&SelectLinearColorFromArray(
     Array Sampling Mode=::NewEnumerator1
     Direct Array Index=&RandomRangeInt(Minimum=0, Maximum=4)
     Color Selection Array=$System.EggColors
     Random Seed=?
    )
    Lifetime=&RandomRangeFloat(Minimum=0.5, Maximum=1.5)
    Position Mode=::NewEnumerator1
    Position Offset=&Lerp_Vector(
     A=$System.Egg 01 Burst Offset
     B=$System.Egg 02 Burst Offset
     Alpha=&MakeFloatFromInt(Int=&SpawnGroup())
    )
    Sprite Size Mode=::NewEnumerator3
    Uniform Sprite Size=&RandomRangeFloat(Minimum=1, Maximum=10)
    UsePositionOffset=true
    Write Material Random=true
   )
   &SampleParticlesFromOtherEmitter(Applied Position Offset:Vector3f=&Lerp_Vector(
     A=$System.Egg 01 Position
     B=$System.Egg 02 Position
     Alpha=&MakeFloatFromInt(Int=&SpawnGroup())
    ))
   &ShapeLocation(MaskSpawnGroup=false, Offset Mode=::NewEnumerator0, Override Seed=false, Sphere Radius=7, UseOwnerScale=true)
   &AddVelocity(Velocity=&RandomRangeVector(Minimum=<-80 -60 150>, Maximum=<100 -150 250>))
  ]
  particleUpdateScript [
   &GravityForce(Gravity=&RandomRangeVector(Minimum=<0 0 -500>, Maximum=<0 0 -200>))
   &Drag(Rotational Drag=0, Use Linear Drag=true, Use Rotational Drag=true)
   &CurlNoiseForce(Bias Noise=false, Noise Frequency=100, Noise Strength=100, Pan Noise=false)
   &SolveForcesAndVelocity(Clamp Velocity=false, Limit Acceleration=false)
   &ParticleState()
   &ScaleSpriteSize(
    Uniform Curve Scale=&Multiply_Float(
     A=&RandomRangeFloat(Minimum=3, Maximum=7)
     B=&RandomRangeFloat(Minimum=0.5, Maximum=1, Evaluation Type=::NewEnumerator1)
    )
    Uniform Curve Sprite Scale=DataInterface() {propertyValues {Curve {keys [
        <at 0 0 RCIM_Cubic>
        {arriveTangent=-2.9089, interpMode=RCIM_Cubic, leaveTangent=-2.9089, tangentMode=RCTM_User, time=0.2, value=1}
        <at 1 0 RCIM_Cubic>
       ]}}}
   )
   &SpriteRotationRate(Rotation Rate=&RandomRangeFloat(Minimum=-5, Maximum=5))
   &ScaleColor(
    Scale RGB=&Multiply_VectorByFloat(Vector=<1 1 1>, Float=&RandomRangeFloat(Minimum=1, Maximum=4))
    ScaleA=true
    ScaleRGB=true
   )
  ]
  rendererClasses [&NiagaraSpriteRendererProperties]
  renderers [{rendererClass=&NiagaraSpriteRendererProperties}]
 }
 EggBurst_SmokeCloud {
  emitterUpdateScript [
   &EmitterState()
   &SpawnBurst_Instantaneous(
    Spawn Count=&MakeCustomIntFromBool(Bool=$System.bIsMoving, True INT=3, False INT=0)
    Spawn Time=&Add_Float(A=$System.Egg Burst Time, B=-0.1)
    Use Spawn Probability=false
    UseLoopCountLimit=false
   ) {enabled=false}
   &SpawnParticlesFromOtherEmitter(
    Attribute Reader=DataInterface() {propertyValues {EmitterBinding {bindingMode=Other, emitterName=Eggs}}}
    Calculate Spawn Rate Per Particle=false
    Maximum Age Spawn Threshold:NiagaraFloat=&Add_Float(A=$System.Egg Burst Time, B=-0.25)
    Minimum Age Spawn Threshold:NiagaraFloat=&Add_Float(A=$System.Egg Burst Time, B=-0.35)
    Spawn Rate=55
    Use Maximum Age=true
    Use Minimum Age=true
   )
  ]
  particleSpawnScript [
   &FN_SetSpawnGroup(Number of Spawn Groups:NiagaraInt32=&Add_Integer(A=$System.EggVisibility, B=1))
   &InitializeParticle(
    Color=$System.Egg 01 Primary Color
    Lifetime Max=0.9
    Lifetime Min=0.55
    Lifetime Mode=::NewEnumerator1
    Position Offset=&Lerp_Vector(
     A=$System.Egg 01 Burst Offset
     B=$System.Egg 02 Burst Offset
     Alpha=&MakeFloatFromInt(Int=&SpawnGroup())
    )
    Sprite Size Mode=::NewEnumerator1
    Sprite UV Mode=::NewEnumerator1
    Uniform Sprite Size Max=120
    Uniform Sprite Size Min=45
    UsePositionOffset=true
    Write Material Random=true
   )
   &SampleParticlesFromOtherEmitter(
    Applied Position Offset:Vector3f=&AddVectorToPosition(
     Position=&Lerp_Vector(
      A=$System.Egg 01 Position
      B=$System.Egg 02 Position
      Alpha=&MakeFloatFromInt(Int=&SpawnGroup())
     )
     Vector=&RandomRangeVector(Minimum=<0 0 19>, Maximum=<0 0 20>)
    )
    Color Sampling=::NewEnumerator1
    Reject Particle If Sampled Index Is Invalid=false
   )
   &ApplyInitialForces()
   &SphereLocation(
    Hemisphere Z=true
    MaskSpawnGroup=false
    Offset=<0 0 -20>
    Override Seed=true
    Random Seed=20
    Sphere Radius=1
    Surface Only=false
    UseNonUniformScale=false
    UseOffset=false
    UseSphereOrientation=false
   )
   &AddVelocity(Velocity=&RandomRangeVector(Minimum=<-20 -100 25>, Maximum=<20 -200 100>))
   &CameraOffset(Camera Offset Amount=&RandomRangeFloat(Minimum=1, Maximum=7))
   /SetVariables(Particles.Color=&LerpLinearColors(
     StartColor=$System.Egg 01 Primary Color
     EndColor=$System.Egg 02 Primary Color
     LerpFactor=&MakeFloatFromInt(Int=&SpawnGroup())
    )) {isSetParameters=true}
  ]
  particleUpdateScript [
   &ParticleState()
   &ScaleSpriteSize(
    Non-Uniform Curve Scale=<1.2 1.2>
    Non-Uniform Curve Sprite Scale=DataInterface() {propertyValues {
      XCurve {keys [
        {arriveTangent=2.2927, interpMode=RCIM_Cubic, leaveTangent=2.2927, tangentMode=RCTM_User, time=0, value=0}
        {arriveTangent=0.0224, interpMode=RCIM_Cubic, leaveTangent=0.0224, tangentMode=RCTM_User, time=1, value=1}
       ]}
      YCurve {keys [
        {arriveTangent=2.2927, interpMode=RCIM_Cubic, leaveTangent=2.2927, tangentMode=RCTM_User, time=0, value=0}
        {arriveTangent=0.0224, interpMode=RCIM_Cubic, leaveTangent=0.0224, tangentMode=RCTM_User, time=1, value=1}
       ]}
     }}
    Scale Sprite Size Mode=::NewEnumerator3
   )
   &AccelerationForce(Acceleration=&VectorFromCurve(
     VectorCurve=DataInterface() {propertyValues {
       XCurve {keys [<at 0 0>,<at 1 0>]}
       YCurve {keys [{tangentMode=RCTM_User, time=0, value=1}, {tangentMode=RCTM_User, time=1, value=0}]}
       ZCurve {keys [
         {arriveTangent=-2.1026, interpMode=RCIM_Cubic, leaveTangent=-2.1026, tangentMode=RCTM_User, time=0, value=1}
         {interpMode=RCIM_Cubic, tangentMode=RCTM_User, time=1, value=0}
        ]}
      }}
     Scale Curve=&RandomRangeVector(Minimum=<0 0 0>, Maximum=<0 0 200>)
    ))
   /SetVariables(
    Particles.DMP_DustChunks:Vector4f=&MakeVector4(
     X=&RandomRangeFloat(Minimum=0, Maximum=1)
     Y=1
     Z=&RandomRangeFloat(Minimum=0.2, Maximum=0.3)
     W=&FloatFromCurve(
      FloatCurve=DataInterface() {propertyValues {Curve {keys [{tangentMode=RCTM_User, time=0, value=0}, {tangentMode=RCTM_User, time=1, value=1}]}}}
      Scale Curve=1.2
     )
    )
    Particles.DustChunksColor:LinearColor=&Multiply_LinearColorByFloat(
     Linear Color=&Multiply_LinearColorByFloat(
      Linear Color=&MultiplyLinearColor(
       Linear Color=$Particles.Initial.Color
       Scale Factor=/Niagara/DynamicInputs/UniformRange/V2/RandomRangeVector4(
        Minimum:Vector4f=<1 1 1 0>
        Maximum:Vector4f=<2 2 2 1>
        Evaluation Type:ENiagaraRandomnessEvaluation=::NewEnumerator0
        Recalculate Random Each Loop:NiagaraBool=true
        Randomness Mode:ENiagaraRandomnessMode=::NewEnumerator0
        Random Seed:NiagaraInt32=0
        Fixed Random Seed:NiagaraBool=false
       )
      )
      Float=&RandomRangeFloat(Minimum=0.7, Maximum=1)
     )
     Float=$System.ToD
    )
   ) {isSetParameters=true}
   &ScaleColor(
    Scale Alpha=&FloatFromCurve(
     FloatCurve=DataInterface() {propertyValues {Curve {keys [<at 0 2 RCIM_Cubic>, {tangentMode=RCTM_User, time=1, value=0}]}}}
     Scale Curve=&RandomRangeFloat(Minimum=0.8, Maximum=1)
    )
    Scale RGB=&VectorFromFloat(Value=&RandomRangeFloat(Minimum=1, Maximum=5))
    ScaleA=true
    ScaleRGB=true
   )
   &ScaleColor(
    Scale Alpha=&FloatFromCurve(
     FloatCurve=DataInterface() {propertyValues {Curve {keys [
         {arriveTangent=-0.0888, interpMode=RCIM_Cubic, leaveTangent=-0.0888, tangentMode=RCTM_User, time=0, value=0}
         {arriveTangent=0.12, interpMode=RCIM_Cubic, leaveTangent=0.12, tangentMode=RCTM_User, time=0.15, value=0.9847}
         {arriveTangent=-2.7131, interpMode=RCIM_Cubic, leaveTangent=-2.7131, tangentMode=RCTM_User, time=1, value=0}
        ]}}}
     Scale Curve=&Multiply_Float(A=1, B=$System.ToD)
    )
    Scale RGB=&VectorFromFloat(Value=&Multiply_Float(A=1, B=$System.ToD))
    ScaleA=true
    ScaleRGB=true
   ) {module=ScaleColor003}
   &GravityForce(Gravity=&VectorFromCurve(
     VectorCurve=DataInterface() {propertyValues {
       XCurve {keys [<at 0 0>,<at 1 0>]}
       YCurve {keys [<at 0 0>,<at 1 0>]}
       ZCurve {keys [
         {arriveTangent=1.0661, interpMode=RCIM_Cubic, leaveTangent=1.0661, tangentMode=RCTM_User, time=0, value=0}
         {interpMode=RCIM_Cubic, tangentMode=RCTM_User, time=1, value=1}
        ]}
      }}
     Scale Curve=&RandomRangeVector(Minimum=<0 0 -300>, Maximum=<0 0 -600>)
    ))
   &Drag(
    Drag=&FloatFromCurve(
     FloatCurve=DataInterface() {propertyValues {Curve {keys [
         {interpMode=RCIM_Cubic, tangentMode=RCTM_User, time=0, value=0}
         {interpMode=RCIM_Cubic, tangentMode=RCTM_User, time=0.7, value=0.9}
         {tangentMode=RCTM_User, time=1, value=0}
        ]}}}
     Scale Curve=4
    )
    Use Linear Drag=true
    Use Rotational Drag=true
   ) {module=Drag002}
   &SolveForcesAndVelocity(Clamp Velocity=false, Limit Acceleration=false)
   &DynamicMaterialParameters(
    Index 0 Param 1=&Add_Float(
     A=&RandomRangeFloat(Minimum=0, Maximum=1)
     B=&FloatFromCurve(
      FloatCurve=DataInterface() {propertyValues {Curve {keys [
          {arriveTangent=2.7533, interpMode=RCIM_Cubic, leaveTangent=1.8958, tangentMode=RCTM_User, time=0, value=1}
          {arriveTangent=0.8292, interpMode=RCIM_Cubic, leaveTangent=0.8292, tangentMode=RCTM_User, time=0.321, value=1.3453}
          {arriveTangent=0.0529, interpMode=RCIM_Cubic, leaveTangent=0.2021, tangentMode=RCTM_User, time=1, value=1.5}
         ]}}}
      CurveIndex=&Add_Float(A=&Multiply_Float(A=1, B=0.5), B=$Particles.NormalizedAge)
     )
    )
    Index 0 Param 2=&FloatFromCurve(FloatCurve=DataInterface() {propertyValues {Curve {keys [
         {arriveTangent=0.0453, interpMode=RCIM_Cubic, tangentMode=RCTM_Break, time=0, value=-1}
         {arriveTangent=0.0604, interpMode=RCIM_Cubic, leaveTangent=0.0604, tangentMode=RCTM_User, time=1, value=-1}
        ]}}})
    Index 0 Param 3=0.5
    Index 0 Param 4=&FloatFromCurve(
     FloatCurve=DataInterface() {propertyValues {Curve {keys [<at 0.5 1>,<at 1 0>]}}}
     Scale Curve=&FNTimeOfDayScalar(
      Night Scale=2
      Morning Scale=2
      Day Scale=&FN_FrontEndFloatModifier(Gameplay=1, Front End=0.5)
      Evening Scale=3
     )
    )
    Index 2 Param 1=&FloatFromCurve(FloatCurve=DataInterface() {propertyValues {Curve {keys [
         {interpMode=RCIM_Cubic, leaveTangent=4.2485, tangentMode=RCTM_Break, time=0, value=0}
         {arriveTangent=1.0548, interpMode=RCIM_Cubic, leaveTangent=1.0548, tangentMode=RCTM_User, time=0.17, value=0.5715}
         <at 1 1>
        ]}}})
    Index 2 Param 2=&FloatFromCurve(
     FloatCurve=DataInterface() {propertyValues {Curve {keys [
         {arriveTangent=1.9056, interpMode=RCIM_Cubic, leaveTangent=1.9056, tangentMode=RCTM_User, time=0, value=0.2}
         {arriveTangent=0.0259, interpMode=RCIM_Cubic, leaveTangent=0.0259, tangentMode=RCTM_User, time=1, value=1.4}
        ]}}}
     Scale Curve=&RandomRangeFloat(Minimum=0.2, Maximum=0.5)
    )
    Index 2 Param 3=&FloatFromCurve(FloatCurve=DataInterface() {propertyValues {Curve {keys [
         {interpMode=RCIM_Cubic, leaveTangent=2.3064, tangentMode=RCTM_Break, time=0, value=1}
         {interpMode=RCIM_Cubic, leaveTangent=11.1233, tangentMode=RCTM_Break, time=1, value=2}
        ]}}})
    Index 2 Param 4=&FloatFromCurve(
     FloatCurve=DataInterface() {propertyValues {Curve {keys [
         {arriveTangent=-0.6867, interpMode=RCIM_Cubic, leaveTangent=-0.6867, tangentMode=RCTM_User, time=0, value=1}
         {interpMode=RCIM_Cubic, leaveTangent=-1.2956, tangentMode=RCTM_Break, time=1, value=0.75}
        ]}}}
     Scale Curve=&RandomRangeFloat(Minimum=1.3, Maximum=1.6)
    )
    Module.Param0WriteEnabled=true
    Module.Param0WriteEnabled2=true
    Module.Param1WriteEnabled=false
    Module.Param1WriteEnabled2=true
    Module.Param2WriteEnabled=true
    Module.Param2WriteEnabled2=true
    Module.Param3WriteEnabled=true
    Module.Param3WriteEnabled2=true
    Write Parameter Index 2=true
   )
  ]
  rendererClasses [&NiagaraSpriteRendererProperties]
  renderers [
   {rendererClass=&NiagaraSpriteRendererProperties}
   {rendererClass=&NiagaraSpriteRendererProperties, rendererIndex=1}
   {rendererClass=&NiagaraSpriteRendererProperties, rendererIndex=2}
  ]
 }
 Eggs {
  emitterSpawnScript [&InitializeAudioPlayer(Audio Player=DataInterface() {propertyValues {
      Attenuation=null
      Concurrency=null
      MaxPlaysPerTick=1
      SoundToPlay=/BRCosmetics/Sounds/Emotes/EggBounce/MSS_Emote_Foley_EggBounce_SFX_Hatch
      bAllowLoopingOneShotSounds=false
      bLimitPlaysPerTick=true
      bOnlyActiveDuringGameplay=false
      bStopWhenComponentIsDestroyed=true
      ConfigurationUserParameter {parameter {name=None, typeDefHandle {registeredTypeIndex=1}, varData []}}
      ParameterNames []
     }})]
  emitterUpdateScript [
   &EmitterState()
   &SpawnPerUnit(Module.Use Max Movement Threshold=true, Module.Use Movement Tolerance=false, Spawn Spacing=5, Use Spawn Probability=false) {enabled=false}
   &SpawnBurst_Instantaneous(Spawn Probability=0.7, Spawn Time=0.02, Use Spawn Probability=true, UseLoopCountLimit=false)
  ]
  particleSpawnScript [
   &InitializeParticle(
    Color=$System.Egg 01 Primary Color
    Lifetime=$System.Egg Burst Time
    Mesh Scale Mode=::NewEnumerator3
    Mesh Uniform Scale=0.24
    Position Offset=$System.Egg 01 Position
    Position Offset Coordinate Space=::Local
    UsePositionOffset=true
    Write Material Random=true
   )
   &InitialMeshOrientation(
    Mesh Orientation Mode=::NewEnumerator6
    Orientation Quaternion:Quat4f=$System.Egg 01 Orientation
    Rotation Coordinate Space=::NewEnumerator0
    Use Rotation Vector=false
   )
   &DynamicMaterialParameters(
    Index 0 Vector 4:Vector4f=$System.Egg 01 Mask Selection
    Index 1 Linear Color:LinearColor=$System.Egg 01 Secondary Color
    Index 2 Param 1=&FloatFromCurve(FloatCurve=DataInterface() {propertyValues {Curve {keys [<at 0 0>,<at 0.4 0>,<at 0.6 1>,<at 1 1>]}}})
    Module.Index2Float0Write=true
    Module.Index2Float1Write=false
    Module.Index2Float2Write=false
    Module.Index2Float3Write=false
    Write Parameter Index 0=::NewEnumerator4
    Write Parameter Index 1=::NewEnumerator5
    Write Parameter Index 2=::NewEnumerator1
   )
   /SetVariables(
    Particles.Egg 02 Mask Select:Vector4f=$System.Egg 02 Mask Selection
    Particles.Egg 02 Orientation:Quat4f=$System.Egg 02 Orientation
    Particles.Egg 02 Position:NiagaraPosition=&AddVectorToPosition(Position=$Particles.Position, Vector=Linked() {linkedVariable {name=System.Egg 02 Position, type=NiagaraPosition}})
    Particles.Egg 02 Primary Color:LinearColor=$System.Egg 02 Primary Color
    Particles.Egg 02 Secondary Color:Vector4f=&MakeVector4_FromLinearColor(LinearColor=$System.Egg 02 Secondary Color)
   ) {isSetParameters=true}
   &AddVelocity(Velocity=<0 -200 0>)
  ]
  particleUpdateScript [
   &SolveForcesAndVelocity(Clamp Velocity=false, Limit Acceleration=false)
   &ParticleState()
   /Niagara/Modules/Audio/PlayAudio(Play Audio Trigger:NiagaraBool=&SetBoolByFloatComparison(A=&Subtract_Float(A=$Particles.Lifetime, B=0.5), B=$Particles.Age, Comparison Type=::NewEnumerator6))
   &DynamicMaterialParameters(
    Index 0 Param 1=/Niagara/DynamicInputs/Integer/V2/Round_Float(Float:NiagaraFloat=0)
    Index 0 Param 3=&FloatFromCurve(FloatCurve=DataInterface() {propertyValues {Curve {keys [<at 0 0>,<at 0.863 0>,<at 1 1>]}}}, Scale Curve=50)
    Index 0 Param 4=1
    Index 1 Linear Color:LinearColor=$System.Egg 01 Secondary Color
    Index 2 Param 1=&FloatFromCurve(FloatCurve=DataInterface() {propertyValues {Curve {keys [
         <at 0 0>
         {arriveTangent=-0.5697, interpMode=RCIM_Cubic, leaveTangent=-0.5697, tangentMode=RCTM_User, time=0.63, value=0}
         {arriveTangent=2.9411, tangentMode=RCTM_User, time=0.9, value=1}
        ]}}})
    Module.Index0Float0Write=false
    Module.Index0Float1Write=false
    Module.Index0Float2Write=true
    Module.Index0Float3Write=true
    Module.Index2Float0Write=true
    Module.Index2Float1Write=true
    Module.Index2Float2Write=true
    Module.Index2Float3Write=true
    Write Parameter Index 1=::NewEnumerator5
    Write Parameter Index 2=::NewEnumerator1
   ) {module=DynamicMaterialParameters001}
   &ScaleMeshSize(Scale Factor=&VectorFromCurve(VectorCurve=DataInterface() {propertyValues {
       XCurve {keys [
         <at 0 0>
         {arriveTangent=2.702, interpMode=RCIM_Cubic, leaveTangent=2.702, tangentMode=RCTM_User, time=0.0287, value=1}
         {arriveTangent=-0.0273, interpMode=RCIM_Cubic, leaveTangent=-0.0273, time=0.1037, value=1}
         {arriveTangent=-0.0414, interpMode=RCIM_Cubic, leaveTangent=-0.0414, time=0.9452, value=0.975}
         {arriveTangent=14.2465, interpMode=RCIM_Cubic, leaveTangent=14.2465, time=0.9633, value=0.9644}
         <at 0.982 1.5 RCIM_Cubic>
        ]}
       YCurve {keys [
         <at 0 0>
         {arriveTangent=2.702, interpMode=RCIM_Cubic, leaveTangent=2.702, tangentMode=RCTM_User, time=0.0287, value=1}
         {arriveTangent=-0.0273, interpMode=RCIM_Cubic, leaveTangent=-0.0273, time=0.1037, value=1}
         {arriveTangent=-0.1118, interpMode=RCIM_Cubic, leaveTangent=-0.1118, time=0.9452, value=0.975}
         {arriveTangent=14.2465, interpMode=RCIM_Cubic, leaveTangent=14.2465, time=0.9586, value=0.9044}
         <at 0.982 1.5 RCIM_Cubic>
        ]}
       ZCurve {keys [
         {arriveTangent=21.9904, interpMode=RCIM_Cubic, leaveTangent=21.9904, tangentMode=RCTM_User, time=0, value=0}
         {arriveTangent=2.702, interpMode=RCIM_Cubic, leaveTangent=2.702, tangentMode=RCTM_User, time=0.0448, value=1.1314}
         {arriveTangent=-2.8973, interpMode=RCIM_Cubic, leaveTangent=-2.8973, time=0.0693, value=1.1189}
         {arriveTangent=-1.2096, interpMode=RCIM_Cubic, leaveTangent=-1.2096, time=0.0867, value=1.0099}
         {arriveTangent=-0.0914, interpMode=RCIM_Cubic, leaveTangent=-0.0914, time=0.1202, value=1.0574}
         {arriveTangent=-0.1924, interpMode=RCIM_Cubic, leaveTangent=-0.1924, time=0.1624, value=1.003}
         {arriveTangent=-0.1157, interpMode=RCIM_Cubic, leaveTangent=-0.1157, time=0.2057, value=1.041}
         {arriveTangent=-0.1804, interpMode=RCIM_Cubic, leaveTangent=-0.1804, time=0.2593, value=0.9917}
         {arriveTangent=0.1063, interpMode=RCIM_Cubic, leaveTangent=0.1063, time=0.3032, value=1.0234}
         {arriveTangent=-0.0363, interpMode=RCIM_Cubic, leaveTangent=-0.0363, time=0.3553, value=1.002}
         {arriveTangent=-0.2722, interpMode=RCIM_Cubic, leaveTangent=-0.2722, time=0.9466, value=1}
         {arriveTangent=4.8946, interpMode=RCIM_Cubic, leaveTangent=4.8946, tangentMode=RCTM_User, time=0.9687, value=0.835}
         {arriveTangent=12.542, interpMode=RCIM_Cubic, leaveTangent=12.542, tangentMode=RCTM_User, time=1, value=2}
        ]}
      }}))
   &ScaleColor(
    Scale Alpha=&FloatFromCurve(FloatCurve=DataInterface() {propertyValues {Curve {keys [
         {interpMode=RCIM_Cubic, tangentMode=RCTM_User, time=0, value=1}
         {arriveTangent=-0.0709, interpMode=RCIM_Cubic, leaveTangent=-0.0709, tangentMode=RCTM_User, time=0.8, value=1}
         {arriveTangent=-2.8317, interpMode=RCIM_Cubic, leaveTangent=-2.8317, tangentMode=RCTM_User, time=1, value=0}
        ]}}})
    ScaleA=true
    ScaleRGB=true
   )
   /SetVariables(
    Particles.Egg 02 Mask Select:Vector4f=&MakeVector4(
     X=&Round_Float(Float:NiagaraFloat=&RandomRangeFloat(Minimum=0, Maximum=3))
     Y=0
     Z=&FloatFromCurve(FloatCurve=DataInterface() {propertyValues {Curve {keys [<at 0 0>,<at 0.86 0>,<at 1 1>]}}}, Scale Curve=50)
     W=1
    )
    Particles.Egg 02 Position Update:NiagaraPosition=&AddVectorToPosition(Position=$Particles.Position, Vector=Linked() {linkedVariable {name=System.Egg 02 Position, type=NiagaraPosition}})
    Particles.Egg 02 Primary Color:LinearColor=&MultiplyLinearColor(
     Linear Color=$Particles.Egg 02 Primary Color
     Scale Factor=&MakeVector4(
      X=1
      Y=1
      Z=1
      W=&MakeFloatFromLinearColor(LinearColor=$Particles.Color, Channel:ENiagaraLinearColor_Channels=::NewEnumerator3)
     )
    )
   ) {isSetParameters=true}
  ]
  rendererClasses [&NiagaraMeshRendererProperties]
  renderers [
   {rendererClass=&NiagaraMeshRendererProperties, rendererIndex=0}
   {rendererClass=&NiagaraMeshRendererProperties, rendererIndex=1}
   {rendererClass=&NiagaraMeshRendererProperties, rendererIndex=2}
  ]
 }
 SimpleSpriteBurst {
  emitterUpdateScript [
   &EmitterState(Life Cycle Mode=::NewEnumerator1, Loop Behavior=::NewEnumerator1, Loop Duration=0.5, UseLoopDelay=false)
   &SpawnBurst_Instantaneous(Use Spawn Probability=false, UseLoopCountLimit=false)
   &SpawnBurst_Instantaneous(Spawn Count=7, Spawn Group=1, Use Spawn Probability=false) {module=SpawnBurst_Instantaneous001}
  ]
  particleSpawnScript [
   /SetVariables(Particles.VisibilityTag=&SpawnGroup()) {isSetParameters=true}
   &InitializeParticle(
    Color=&MaskLinearColorBySpawnGroup(
     Masked Linear Color=<rgba 1 1 1 1>
     Passthrough Linear Color=&RandomRangeLinearColor(Minimum=<rgba 0.4323 0.3763 0.2276 0.25>, Maximum=<rgba 0.0957 0.0799 0.0399 1>)
    )
    Lifetime=&MaskFloatBySpawnGroup(Masked Float=0.4, Passthrough Float=&RandomRangeFloat(Minimum=0.25, Maximum=1))
    Sprite Size Mode=::NewEnumerator3
    Uniform Sprite Size=&MaskFloatBySpawnGroup(Masked Float=60, Passthrough Float=&RandomRangeFloat(Minimum=50, Maximum=70))
    UsePositionOffset=false
    Write Material Random=true
   )
   &SpriteFacingAndAlignment(Sprite Alignment=<1 0 0>, Sprite Facing=<0 0 1>, WriteAlignment=true, WriteFacing=true)
   &CameraOffset(Camera Offset Amount=&MaskFloatBySpawnGroup(Masked Float=5, Passthrough Float=0))
   /SetVariables(Particles.FlareOffset=4) {isSetParameters=true}
   &AccelerationForce(Acceleration=&MaskVectorBySpawnGroup(Masked Vector=<0 0 0>, Passthrough Vector=&UniformRangedVector(Minimum=<0 0 0>, Maximum=<0 0 50>)))
   &AddVelocity(Velocity=&MaskVectorBySpawnGroup(
     Masked Vector=<0 -200 0>
     Passthrough Vector=&Add_Vector(A=&UniformRangedVector(Minimum=<-50 -50 -50>, Maximum=<50 50 50>), B=<0 -145 0>)
    ))
   &ApplyInitialForces()
   &ShapeLocation(
    MaskSpawnGroup=false
    Override Seed=false
    Sphere Radius=&MaskFloatBySpawnGroup(Masked Float=0, Passthrough Float=3)
    UseOwnerScale=true
   )
  ]
  particleUpdateScript [
   &ParticleState()
   &ScaleColor(
    Scale Alpha=&MaskFloatBySpawnGroup(
     Masked Float=&FloatFromCurve(
      FloatCurve=DataInterface() {propertyValues {Curve {keys [
          {arriveTangent=9.0725, interpMode=RCIM_Cubic, leaveTangent=9.0725, tangentMode=RCTM_User, time=0, value=0}
          {arriveTangent=-2.5249, interpMode=RCIM_Cubic, leaveTangent=-2.5249, tangentMode=RCTM_User, time=0.1408, value=0.9847}
          {arriveTangent=-0.0405, interpMode=RCIM_Cubic, leaveTangent=-0.0405, tangentMode=RCTM_User, time=1, value=0}
         ]}}}
      Scale Curve=0.4
     )
     Passthrough Float=&FloatFromCurve(
      FloatCurve=DataInterface() {propertyValues {Curve {keys [
          {arriveTangent=9.0725, interpMode=RCIM_Cubic, leaveTangent=9.0725, tangentMode=RCTM_User, time=0, value=0}
          {arriveTangent=-0.0962, interpMode=RCIM_Cubic, leaveTangent=-0.0962, tangentMode=RCTM_User, time=0.1408, value=0.9847}
          {arriveTangent=-1.3793, interpMode=RCIM_Cubic, leaveTangent=-1.3793, tangentMode=RCTM_User, time=1, value=0}
         ]}}}
      Scale Curve=&RandomRangeFloat(Minimum=0.8, Maximum=1.5)
     )
    )
    Scale RGB=&MaskVectorBySpawnGroup(Masked Vector=<1 1 1>, Passthrough Vector=&VectorFromFloat(Value=0.4))
    ScaleA=true
    ScaleRGB=true
   )
   /SetVariables(
    Particles.MiniRingColor:LinearColor=&Multiply_LinearColorByFloat(
     Linear Color=$Particles.Initial.Color
     Float=&FloatFromCurve(
      FloatCurve=DataInterface() {propertyValues {Curve {keys [
          {arriveTangent=9.0725, interpMode=RCIM_Cubic, leaveTangent=9.0725, tangentMode=RCTM_User, time=0, value=0}
          {arriveTangent=-2.5249, interpMode=RCIM_Cubic, leaveTangent=-2.5249, tangentMode=RCTM_User, time=0.1408, value=0.9847}
          {arriveTangent=-0.0405, interpMode=RCIM_Cubic, leaveTangent=-0.0405, tangentMode=RCTM_User, time=1, value=0}
         ]}}}
      Scale Curve=0.6
     )
    )
    Particles.MiniRingSize:Vector2f=&Multiply_Vector2D_ByFloat(
     Float=&FloatFromCurve(
      FloatCurve=DataInterface() {propertyValues {Curve {keys [
          {arriveTangent=4.1785, interpMode=RCIM_Cubic, leaveTangent=4.1785, tangentMode=RCTM_User, time=0, value=0}
          {arriveTangent=1.1179, interpMode=RCIM_Cubic, leaveTangent=1.1179, tangentMode=RCTM_User, time=0.2042, value=0.6228}
          {arriveTangent=0.0224, interpMode=RCIM_Cubic, leaveTangent=0.0224, tangentMode=RCTM_User, time=1, value=1}
         ]}}}
      Scale Curve=0.4
     )
     Vector2D=$Particles.Initial.SpriteSize
    )
   ) {isSetParameters=true}
   /SetVariables(
    Particles.FlareColor=&Multiply_LinearColorByFloat(
     Linear Color=&ColorFromCurve(ColorCurve=DataInterface() {propertyValues {
        AlphaCurve {keys [<at 0 1 RCIM_Cubic>,<at 1 0 RCIM_Cubic>]}
        BlueCurve {keys [<at 0 25 RCIM_Cubic>,<at 0.3956 1>,<at 1 1 RCIM_Cubic>]}
        GreenCurve {keys [<at 0 25 RCIM_Cubic>,<at 0.3956 1>,<at 1 1 RCIM_Cubic>]}
        RedCurve {keys [<at 0 25 RCIM_Cubic>,<at 0.3956 1>,<at 1 1 RCIM_Cubic>]}
       }})
     Float=&FloatFromCurve(FloatCurve=DataInterface() {propertyValues {Curve {keys [
          {arriveTangent=-1.1798, interpMode=RCIM_Cubic, leaveTangent=-1.1798, tangentMode=RCTM_User, time=0, value=1}
          {arriveTangent=-1.6667, interpMode=RCIM_Cubic, leaveTangent=-1.6667, time=0.1, value=0.85}
          {arriveTangent=-1.5049, interpMode=RCIM_Cubic, leaveTangent=-1.5049, tangentMode=RCTM_User, time=0.6, value=0}
         ]}}})
    )
    Particles.FlareSize:Vector2f=&Multiply_Vector2D_ByFloat(
     Float=&FloatFromCurve(FloatCurve=DataInterface() {propertyValues {Curve {keys [
          {interpMode=RCIM_Cubic, tangentMode=RCTM_User, time=0, value=0}
          {arriveTangent=6.5121, interpMode=RCIM_Cubic, leaveTangent=6.5121, tangentMode=RCTM_User, time=1, value=5}
         ]}}})
     Vector2D=<25 25>
    )
   ) {isSetParameters=true}
   &ScaleSpriteSize(
    Scale Sprite Size Mode=::NewEnumerator0
    Uniform Scale Factor=&MaskFloatBySpawnGroup(
     Masked Float=&FloatFromCurve(FloatCurve=DataInterface() {propertyValues {Curve {keys [
          {arriveTangent=4.1785, interpMode=RCIM_Cubic, leaveTangent=4.1785, tangentMode=RCTM_User, time=0, value=0}
          {arriveTangent=1.7328, interpMode=RCIM_Cubic, leaveTangent=1.7328, tangentMode=RCTM_User, time=0.2042, value=0.6228}
          {arriveTangent=0.0224, interpMode=RCIM_Cubic, leaveTangent=0.0224, tangentMode=RCTM_User, time=1, value=1.5}
         ]}}})
     Passthrough Float=&FloatFromCurve(FloatCurve=DataInterface() {propertyValues {Curve {keys [
          {arriveTangent=4.1785, interpMode=RCIM_Cubic, leaveTangent=4.1785, tangentMode=RCTM_User, time=0, value=0}
          {arriveTangent=1.1179, interpMode=RCIM_Cubic, leaveTangent=1.1179, tangentMode=RCTM_User, time=0.2042, value=0.6228}
          {arriveTangent=0.0224, interpMode=RCIM_Cubic, leaveTangent=0.0224, tangentMode=RCTM_User, time=1, value=1.3}
         ]}}})
    )
   )
   &Drag(
    Drag=&MaskFloatBySpawnGroup(
     Masked Float=0
     Passthrough Float=&FloatFromCurve(
      FloatCurve=DataInterface() {propertyValues {Curve {keys [
          {arriveTangent=2.8162, interpMode=RCIM_Cubic, leaveTangent=2.8162, tangentMode=RCTM_User, time=0, value=0}
          {interpMode=RCIM_Cubic, tangentMode=RCTM_User, time=0.5, value=1}
          {arriveTangent=-2.8774, tangentMode=RCTM_User, time=1, value=0}
         ]}}}
      Scale Curve=5
     )
    )
    Use Linear Drag=true
    Use Rotational Drag=true
   )
   &AccelerationForce(Acceleration=&MaskVectorBySpawnGroup(
     Masked Vector=<0 0 0>
     Passthrough Vector=&VectorFromCurve(
      VectorCurve=DataInterface() {propertyValues {
        XCurve {keys [<at 0 0>,<at 1 0>]}
        YCurve {keys [{tangentMode=RCTM_User, time=0, value=1}, {tangentMode=RCTM_User, time=1, value=0}]}
        ZCurve {keys [{tangentMode=RCTM_User, time=0, value=0}, {tangentMode=RCTM_User, time=1, value=0}]}
       }}
      Scale Curve=&RandomRangeVector(Minimum=<0 -700 0>, Maximum=<0 -750 0>)
     )
    )) {enabled=false, module=AccelerationForce001}
   &SolveForcesAndVelocity(Clamp Velocity=false, Limit Acceleration=false)
   &CameraOffset(Camera Offset Amount=&MaskFloatBySpawnGroup(Masked Float=$Particles.Initial.CameraOffset, Passthrough Float=10)) {module=CameraOffset001}
   &DynamicMaterialParameters(
    Index 0 Param 1=&FloatFromCurve(FloatCurve=DataInterface() {propertyValues {Curve {keys [
         {arriveTangent=4.1785, interpMode=RCIM_Cubic, leaveTangent=4.1785, tangentMode=RCTM_User, time=0, value=0}
         {arriveTangent=1.1179, interpMode=RCIM_Cubic, leaveTangent=1.1179, tangentMode=RCTM_User, time=0.2042, value=0.6228}
         {arriveTangent=0.0224, interpMode=RCIM_Cubic, leaveTangent=0.0224, tangentMode=RCTM_User, time=1, value=1}
        ]}}})
    Index 0 Param 3=0.5
    Index 0 Param 4=&FloatFromCurve(
     FloatCurve=DataInterface() {propertyValues {Curve {keys [<at 0.5 1>,<at 1 0>]}}}
     Scale Curve=&FNTimeOfDayScalar(
      Night Scale=2
      Morning Scale=2
      Day Scale=&FN_FrontEndFloatModifier(Gameplay=1, Front End=0.5)
      Evening Scale=3
     )
    )
    Index 2 Param 1=&FloatFromCurve(FloatCurve=DataInterface() {propertyValues {Curve {keys [{interpMode=RCIM_Cubic, tangentMode=RCTM_User, time=0, value=0}, {interpMode=RCIM_Cubic, tangentMode=RCTM_User, time=1, value=1}]}}})
    Index 2 Param 2=&FloatFromCurve(
     FloatCurve=DataInterface() {propertyValues {Curve {keys [
         {arriveTangent=1.9056, interpMode=RCIM_Cubic, leaveTangent=1.9056, tangentMode=RCTM_User, time=0, value=0.2}
         {arriveTangent=0.0259, interpMode=RCIM_Cubic, leaveTangent=0.0259, tangentMode=RCTM_User, time=1, value=1.4}
        ]}}}
     Scale Curve=&RandomRangeFloat(Minimum=0.2, Maximum=0.5)
    )
    Index 2 Param 3=&FloatFromCurve(FloatCurve=DataInterface() {propertyValues {Curve {keys [
         {interpMode=RCIM_Cubic, leaveTangent=2.3064, tangentMode=RCTM_Break, time=0, value=1}
         {interpMode=RCIM_Cubic, leaveTangent=11.1233, tangentMode=RCTM_Break, time=1, value=2}
        ]}}})
    Index 2 Param 4=&FloatFromCurve(FloatCurve=DataInterface() {propertyValues {Curve {keys [
         {arriveTangent=-0.6867, interpMode=RCIM_Cubic, leaveTangent=-0.6867, tangentMode=RCTM_User, time=0, value=1}
         {interpMode=RCIM_Cubic, leaveTangent=-1.2956, tangentMode=RCTM_Break, time=1, value=0.75}
        ]}}})
    Module.Param0WriteEnabled=true
    Module.Param0WriteEnabled2=true
    Module.Param1WriteEnabled=false
    Module.Param1WriteEnabled2=true
    Module.Param2WriteEnabled=true
    Module.Param2WriteEnabled2=true
    Module.Param3WriteEnabled=true
    Module.Param3WriteEnabled2=true
    Write Parameter Index 2=true
   )
   /SetVariables(
    Particles.DMP001:Vector4f=&MakeVector4(X=0, Y=0, Z=0, W=0)
    Particles.DMP002:Vector4f=&MakeVector4(X=0, Y=0, Z=0, W=0)
    Particles.DMP003:Vector4f=&MakeVector4(X=0, Y=0, Z=0, W=0)
    Particles.DMP_0:Vector4f=&MakeVector4(X=0, Y=0, Z=0, W=0)
   ) {isSetParameters=true}
  ]
  rendererClasses [&NiagaraSpriteRendererProperties]
  renderers [
   {rendererClass=&NiagaraSpriteRendererProperties}
   {rendererClass=&NiagaraSpriteRendererProperties, rendererIndex=1}
   {rendererClass=&NiagaraSpriteRendererProperties, rendererIndex=2}
  ]
 }
}
summary {
 systemName=NS_Emote_EggBounce_Eggs_FE
 emitters [
  {emitterName=Eggs, rendererClasses [&NiagaraMeshRendererProperties]}
  {emitterName=EggBurst_Glitter, rendererClasses [&NiagaraSpriteRendererProperties]}
  {emitterName=EggBurst_SmokeCloud, rendererClasses [&NiagaraSpriteRendererProperties]}
  {emitterName=SimpleSpriteBurst, rendererClasses [&NiagaraSpriteRendererProperties]}
  {emitterName=Debris_Sprite, simTarget=GPUComputeSim, rendererClasses [&NiagaraSpriteRendererProperties]}
 ]
 userVariables [
  User.bIsFrontEnd {type=NiagaraBool}
  User.bIsFrontEndPreview {type=NiagaraBool}
 ]
}
