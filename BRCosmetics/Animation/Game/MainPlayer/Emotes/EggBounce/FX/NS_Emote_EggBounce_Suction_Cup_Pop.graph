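# Niagara one-shot effect for the suction cup pop in the EggBounce emote.
# A GPU debris emitter bursts three spawn groups from a torus shape, with per-group gravity, drag, speed stretching and fade.
# A second emitter layers a single expanding sprite with seven drifting ones, driving ring, flare and dynamic material curves.
# Spawn group sets VisibilityTag and masks most inputs; front-end and time-of-day inputs adjust color and material intensity.
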NiagaraSystem /BRCosmetics/Animation/Game/MainPlayer/Emotes/EggBounce/FX/NS_Emote_EggBounce_Suction_Cup_Pop
dependencies {
 usedDataInterfaces [&NiagaraDataInterfaceCurve, &NiagaraDataInterfaceColorCurve, &NiagaraDataInterfaceVector2DCurve]
 usedDynamicInputs [&MakeFloatFromVector2d, &Multiply_VectorByFloat]
 usedModules [&SystemState]
}
emitters {
 Debris_Sprite001 {
  simTarget=GPUComputeSim
  emitterUpdateScript [
   &EmitterState(Life Cycle Mode=::NewEnumerator1, Loop Behavior=::NewEnumerator1, Loop Delay=0.05, Loop Duration=0.5, UseLoopDelay=false)
   &SpawnBurst_Instantaneous(
    Spawn Count=&RandomRangeInt(Minimum=5, Maximum=7)
    Use Spawn Probability=false
    UseLoopCountLimit=false
   )
   &SpawnBurst_Instantaneous(
    Spawn Count=&RandomRangeInt(Minimum=3, Maximum=4)
    Spawn Group=1
    Use Spawn Probability=false
    UseLoopCountLimit=false
   ) {module=SpawnBurst_Instantaneous001}
   &SpawnBurst_Instantaneous(Spawn Count=&RandomRangeInt(Minimum=1, Maximum=10), Spawn Group=2, Use Spawn Probability=false) {module=SpawnBurst_Instantaneous002}
  ]
  particleSpawnScript [
   /SetVariables(Particles.VisibilityTag=&SpawnGroup()) {isSetParameters=true}
   &InitializeParticle(
    Color=&MaskLinearColorBySpawnGroup(
     Masked Linear Color=&MakeCustomLinearColorFromBool(
      Bool=$User.bIsFrontEndPreview
      True Color=&RandomRangeLinearColor(Minimum=<rgba 0.3177 0.1972 0.1605 0.35>, Maximum=<rgba 0.3021 0.2182 0.1731 0.5>)
      False Color=&RandomRangeLinearColor(Minimum=<rgba 0.375 0.2073 0.1562 0.35>, Maximum=<rgba 0.375 0.2328 0.1562 0.5>)
     )
     Passthrough Linear Color=&MaskLinearColorBySpawnGroup(Masked Linear Color=<rgba 1 1 1 1>, Passthrough Linear Color=<rgba 0.4167 0.4167 0.4167 0.4957>, Spawn Group Mask=1)
    )
    Lifetime=&MaskFloatBySpawnGroup(
     Masked Float=&RandomRangeFloat(Minimum=0.3, Maximum=1)
     Passthrough Float=&MaskFloatBySpawnGroup(
      Masked Float=&MaskFloatBySpawnGroup(Masked Float=0.2, Passthrough Float=0.4)
      Passthrough Float=0.3
      Spawn Group Mask=1
     )
    )
    Mass Mode=::NewEnumerator1
    Position Offset=&MaskVectorBySpawnGroup(
     Masked Vector=<0 0 0>
     Passthrough Vector=&MaskVectorBySpawnGroup(Masked Vector=<0 0 -7>, Passthrough Vector=<0 0 -8>, Spawn Group Mask=1)
    )
    Sprite Rotation Mode=::NewEnumerator2
    Sprite Size=&MaskVector2dBySpawnGroup(
     Masked Vector2D=&RandomRangeVector2D(Minimum=<5 5>, Maximum=<10 10>)
     Passthrough Vector2D=&MaskVector2dBySpawnGroup(
      Masked Vector2D=<1.5 1.5>
      Passthrough Vector2D=&RandomRangeVector2D(
       Minimum=&Multiply_Vector2D_ByFloat(Float=0.6, Vector2D=<40 62>)
       Maximum=&Multiply_Vector2D_ByFloat(Float=0.65, Vector2D=<20 50>)
      )
      Spawn Group Mask=1
     )
     Spawn Group Mask=0
    )
    Sprite Size Mode=::NewEnumerator4
    UsePositionOffset=true
    Write Material Random=true
   )
   &ShapeLocation(
    Handle Radius=1
    Large Radius=&MaskFloatBySpawnGroup(
     Masked Float=3
     Passthrough Float=&MaskFloatBySpawnGroup(Masked Float=0, Passthrough Float=4, Spawn Group Mask=1)
    )
    MaskSpawnGroup=false
    Shape Primitive=::NewEnumerator3
    Torus Distribution Mode:ENiagaraTorusDistributionMode=::NewEnumerator1
    U Position=&MaskFloatBySpawnGroup(
     Masked Float=&RandomRangeFloat(Minimum=0, Maximum=1)
     Passthrough Float=&MaskFloatBySpawnGroup(
      Masked Float=&RandomRangeFloat(Minimum=0, Maximum=1)
      Passthrough Float=&ReturnNormalizedExecIndex()
      Spawn Group Mask=1
     )
    )
    UseOwnerScale=true
   )
   &InitialMeshOrientation(
    Mesh Axis To Orient=<0 1 0>
    Mesh Orientation Mode=::NewEnumerator2
    Orientation Vector=$Particles.ShapeLocation.ShapeTangent
    Use Rotation Vector=true
   )
   &AddVelocity(
    Module.Use Falloff Distance=false
    Origin Offset=&MaskVectorBySpawnGroup(
     Masked Vector=&RandomRangeVector(Minimum=<0 0 -1>, Maximum=<0 0 -5>)
     Passthrough Vector=&MaskVectorBySpawnGroup(
      Masked Vector=&RandomRangeVector(Minimum=<-10 -10 -8>, Maximum=<10 10 -50>)
      Passthrough Vector=<0 0 -10>
      Spawn Group Mask=1
     )
    )
    Override Seed=false
    Velocity Mode=::NewEnumerator1
    Velocity Origin=&AddVectorToPosition(
     Position=$Engine.Emitter.SimulationPosition
     Vector=&MaskVectorBySpawnGroup(
      Masked Vector=<0 0 0>
      Passthrough Vector=&MaskVectorBySpawnGroup(Masked Vector=<0 0 0>, Passthrough Vector=<0 0 0>, Spawn Group Mask=1)
     )
    )
    Velocity Speed=&MaskFloatBySpawnGroup(
     Masked Float=&RandomRangeFloat(Minimum=150, Maximum=300)
     Passthrough Float=&MaskFloatBySpawnGroup(Masked Float=&RandomRangeFloat(Minimum=150, Maximum=200), Passthrough Float=75, Spawn Group Mask=1)
    )
   )
   &CameraOffset(Camera Offset Amount=&MaskFloatBySpawnGroup(
     Masked Float=0
     Passthrough Float=&MaskFloatBySpawnGroup(Masked Float=0, Passthrough Float=1, Spawn Group Mask=1)
    ))
  ]
  particleUpdateScript [
   &ParticleState()
   &GravityForce(Gravity=&MaskVectorBySpawnGroup(
     Masked Vector=<0 0 -980>
     Passthrough Vector=&MaskVectorBySpawnGroup(Masked Vector=<0 0 0>, Passthrough Vector=<0 0 0>, Spawn Group Mask=1)
    ))
   &Drag(
    Drag=&MaskFloatBySpawnGroup(
     Masked Float=1
     Passthrough Float=&MaskFloatBySpawnGroup(
      Masked Float=&FloatFromCurve(
       FloatCurve=DataInterface() {propertyValues {Curve {keys [
           {interpMode=RCIM_Cubic, tangentMode=RCTM_User, time=0, value=0}
           {arriveTangent=3.2375, interpMode=RCIM_Cubic, leaveTangent=3.2375, tangentMode=RCTM_User, time=1, value=1}
          ]}}}
       Scale Curve=102.9192
      )
      Passthrough Float=0
      Spawn Group Mask=1
     )
    )
    Use Linear Drag=true
    Use Rotational Drag=true
   )
   &SolveForcesAndVelocity(Clamp Velocity=false, Limit Acceleration=false)
   &ScaleSpriteSizeBySpeed(
    Max Scale Factor=&MaskVector2dBySpawnGroup(
     Masked Vector2D=<1 1>
     Passthrough Vector2D=&MaskVector2dBySpawnGroup(Masked Vector2D=<0.4 20>, Passthrough Vector2D=<1 1>, Spawn Group Mask=1)
     Spawn Group Mask=0
    )
    Min Scale Factor=<1 0.75>
    Velocity Threshold=500
   )
   &ScaleSpriteSize(Scale Factor=&MaskVector2dBySpawnGroup(
     Masked Vector2D=&Vector2DFromFloat(Value=&FloatFromCurve(FloatCurve=DataInterface() {propertyValues {Curve {keys [<at 0.5 1 RCIM_Cubic>, {arriveTangent=-3.7562, interpMode=RCIM_Cubic, leaveTangent=-3.7562, tangentMode=RCTM_User, time=1, value=0}]}}}))
     Passthrough Vector2D=&MaskVector2dBySpawnGroup(
      Masked Vector2D=&Vector2DFromFloat(Value=&FloatFromCurve(FloatCurve=DataInterface() {propertyValues {Curve {keys [<at 0.5 1 RCIM_Cubic>, {arriveTangent=-3.7562, interpMode=RCIM_Cubic, leaveTangent=-3.7562, tangentMode=RCTM_User, time=1, value=0}]}}}))
      Passthrough Vector2D=&Vector2DFromCurve(
       Vector2Curve=DataInterface() {propertyValues {XCurve {keys [<at 0 0>,<at 1 1>]}, YCurve {keys [<at 0 0>,<at 0.2 0.775>,<at 1 0>]}}}
       ScaleCurve=&Multiply_Vector2D_ByFloat(Float=&RandomRangeFloat(Minimum=0.8, Maximum=1), Vector2D=<1 1>)
      )
      Spawn Group Mask=1
     )
     Spawn Group Mask=0
    ))
   &ScaleColor(
    Scale Alpha=&MaskFloatBySpawnGroup(
     Masked Float=1
     Passthrough Float=&MaskFloatBySpawnGroup(
      Masked Float=&FloatFromCurve(FloatCurve=DataInterface() {propertyValues {Curve {keys [{interpMode=RCIM_Cubic, tangentMode=RCTM_User, time=0, value=1}, {interpMode=RCIM_Cubic, tangentMode=RCTM_User, time=1, value=0}]}}})
      Passthrough Float=&FloatFromCurve(
       FloatCurve=DataInterface() {propertyValues {Curve {keys [
           {arriveTangent=-4.0245, interpMode=RCIM_Cubic, leaveTangent=-4.0245, tangentMode=RCTM_User, time=0, value=1.5}
           {arriveTangent=-1.5, interpMode=RCIM_Cubic, leaveTangent=-1.5, time=0.2, value=0.8}
           <at 1 0>
          ]}}}
       Scale Curve=&RandomRangeFloat(Minimum=0.05, Maximum=0.7)
      )
      Spawn Group Mask=1
     )
    )
    Scale RGB=&MaskVectorBySpawnGroup(
     Masked Vector=<1 1 1>
     Passthrough Vector=&MaskVectorBySpawnGroup(Masked Vector=<1 1 1>, Passthrough Vector=&VectorFromFloat(Value=0.5), Spawn Group Mask=1)
    )
    ScaleA=true
    ScaleRGB=true
   )
  ]
  rendererClasses [&NiagaraSpriteRendererProperties]
  renderers [
   {rendererClass=&NiagaraSpriteRendererProperties}
   {rendererClass=&NiagaraSpriteRendererProperties, rendererIndex=1}
   {rendererClass=&NiagaraSpriteRendererProperties, rendererIndex=2}
  ]
 }
 SimpleSpriteBurst001 {
  emitterUpdateScript [
   &EmitterState(Life Cycle Mode=::NewEnumerator1, Loop Behavior=::NewEnumerator1, Loop Duration=0.5, UseLoopDelay=false)
   &SpawnBurst_Instantaneous(Use Spawn Probability=false, UseLoopCountLimit=false)
   &SpawnBurst_Instantaneous(Spawn Count=7, Spawn Group=1, Use Spawn Probability=false) {module=SpawnBurst_Instantaneous001}
  ]
  particleSpawnScript [
   /SetVariables(Particles.VisibilityTag=&SpawnGroup()) {isSetParameters=true}
   &InitializeParticle(
    Color=&MaskLinearColorBySpawnGroup(
     Masked Linear Color=<rgba 1 1 1 1>
     Passthrough Linear Color=&RandomRangeLinearColor(Minimum=<rgba 0.4323 0.3763 0.2276 0.25>, Maximum=<rgba 0.0957 0.0799 0.0399 1>)
    )
    Lifetime=&MaskFloatBySpawnGroup(Masked Float=0.4, Passthrough Float=&RandomRangeFloat(Minimum=0.25, Maximum=1))
    Sprite Size Mode=::NewEnumerator3
    Uniform Sprite Size=&MaskFloatBySpawnGroup(Masked Float=60, Passthrough Float=&RandomRangeFloat(Minimum=50, Maximum=70))
    UsePositionOffset=false
    Write Material Random=true
   )
   &SpriteFacingAndAlignment(Sprite Alignment=<1 0 0>, Sprite Facing=<0 0 1>, WriteAlignment=true, WriteFacing=true)
   &CameraOffset(Camera Offset Amount=&MaskFloatBySpawnGroup(Masked Float=5, Passthrough Float=0))
   /SetVariables(Particles.FlareOffset=4) {isSetParameters=true}
   &AccelerationForce(Acceleration=&MaskVectorBySpawnGroup(Masked Vector=<0 0 0>, Passthrough Vector=&UniformRangedVector(Minimum=<0 0 0>, Maximum=<0 0 50>)))
   &AddVelocity(Velocity=&MaskVectorBySpawnGroup(Masked Vector=<0 0 0>, Passthrough Vector=&UniformRangedVector(Minimum=<-50 -50 -50>, Maximum=<50 50 50>)))
   &ApplyInitialForces()
   &ShapeLocation(
    MaskSpawnGroup=false
    Override Seed=false
    Sphere Radius=&MaskFloatBySpawnGroup(Masked Float=0, Passthrough Float=3)
    UseOwnerScale=true
   )
  ]
  particleUpdateScript [
   &ParticleState()
   &ScaleColor(
    Scale Alpha=&MaskFloatBySpawnGroup(
     Masked Float=&FloatFromCurve(
      FloatCurve=DataInterface() {propertyValues {Curve {keys [
          {arriveTangent=9.0725, interpMode=RCIM_Cubic, leaveTangent=9.0725, tangentMode=RCTM_User, time=0, value=0}
          {arriveTangent=-2.5249, interpMode=RCIM_Cubic, leaveTangent=-2.5249, tangentMode=RCTM_User, time=0.1408, value=0.9847}
          {arriveTangent=-0.0405, interpMode=RCIM_Cubic, leaveTangent=-0.0405, tangentMode=RCTM_User, time=1, value=0}
         ]}}}
      Scale Curve=0.4
     )
     Passthrough Float=&FloatFromCurve(
      FloatCurve=DataInterface() {propertyValues {Curve {keys [
          {arriveTangent=9.0725, interpMode=RCIM_Cubic, leaveTangent=9.0725, tangentMode=RCTM_User, time=0, value=0}
          {arriveTangent=-0.0962, interpMode=RCIM_Cubic, leaveTangent=-0.0962, tangentMode=RCTM_User, time=0.1408, value=0.9847}
          {arriveTangent=-1.3793, interpMode=RCIM_Cubic, leaveTangent=-1.3793, tangentMode=RCTM_User, time=1, value=0}
         ]}}}
      Scale Curve=&RandomRangeFloat(Minimum=0.8, Maximum=1.5)
     )
    )
    Scale RGB=&MaskVectorBySpawnGroup(Masked Vector=<1 1 1>, Passthrough Vector=&VectorFromFloat(Value=0.4))
    ScaleA=true
    ScaleRGB=true
   )
   /SetVariables(
    Particles.MiniRingColor:LinearColor=&Multiply_LinearColorByFloat(
     Linear Color=$Particles.Initial.Color
     Float=&FloatFromCurve(
      FloatCurve=DataInterface() {propertyValues {Curve {keys [
          {arriveTangent=9.0725, interpMode=RCIM_Cubic, leaveTangent=9.0725, tangentMode=RCTM_User, time=0, value=0}
          {arriveTangent=-2.5249, interpMode=RCIM_Cubic, leaveTangent=-2.5249, tangentMode=RCTM_User, time=0.1408, value=0.9847}
          {arriveTangent=-0.0405, interpMode=RCIM_Cubic, leaveTangent=-0.0405, tangentMode=RCTM_User, time=1, value=0}
         ]}}}
      Scale Curve=0.6
     )
    )
    Particles.MiniRingSize:Vector2f=&Multiply_Vector2D_ByFloat(
     Float=&FloatFromCurve(
      FloatCurve=DataInterface() {propertyValues {Curve {keys [
          {arriveTangent=4.1785, interpMode=RCIM_Cubic, leaveTangent=4.1785, tangentMode=RCTM_User, time=0, value=0}
          {arriveTangent=1.1179, interpMode=RCIM_Cubic, leaveTangent=1.1179, tangentMode=RCTM_User, time=0.2042, value=0.6228}
          {arriveTangent=0.0224, interpMode=RCIM_Cubic, leaveTangent=0.0224, tangentMode=RCTM_User, time=1, value=1}
         ]}}}
      Scale Curve=0.4
     )
     Vector2D=$Particles.Initial.SpriteSize
    )
   ) {isSetParameters=true}
   /SetVariables(
    Particles.FlareColor=&Multiply_LinearColorByFloat(
     Linear Color=&ColorFromCurve(ColorCurve=DataInterface() {propertyValues {
        AlphaCurve {keys [<at 0 1 RCIM_Cubic>,<at 1 0 RCIM_Cubic>]}
        BlueCurve {keys [<at 0 25 RCIM_Cubic>,<at 0.3956 1>,<at 1 1 RCIM_Cubic>]}
        GreenCurve {keys [<at 0 25 RCIM_Cubic>,<at 0.3956 1>,<at 1 1 RCIM_Cubic>]}
        RedCurve {keys [<at 0 25 RCIM_Cubic>,<at 0.3956 1>,<at 1 1 RCIM_Cubic>]}
       }})
     Float=&FloatFromCurve(FloatCurve=DataInterface() {propertyValues {Curve {keys [
          {arriveTangent=-1.1798, interpMode=RCIM_Cubic, leaveTangent=-1.1798, tangentMode=RCTM_User, time=0, value=1}
          {arriveTangent=-1.6667, interpMode=RCIM_Cubic, leaveTangent=-1.6667, time=0.1, value=0.85}
          {arriveTangent=-1.5049, interpMode=RCIM_Cubic, leaveTangent=-1.5049, tangentMode=RCTM_User, time=0.6, value=0}
         ]}}})
    )
    Particles.FlareSize:Vector2f=&Multiply_Vector2D_ByFloat(
     Float=&FloatFromCurve(FloatCurve=DataInterface() {propertyValues {Curve {keys [
          {interpMode=RCIM_Cubic, tangentMode=RCTM_User, time=0, value=0}
          {arriveTangent=6.5121, interpMode=RCIM_Cubic, leaveTangent=6.5121, tangentMode=RCTM_User, time=1, value=5}
         ]}}})
     Vector2D=<25 25>
    )
   ) {isSetParameters=true}
   &ScaleSpriteSize(
    Scale Sprite Size Mode=::NewEnumerator0
    Uniform Scale Factor=&MaskFloatBySpawnGroup(
     Masked Float=&FloatFromCurve(FloatCurve=DataInterface() {propertyValues {Curve {keys [
          {arriveTangent=4.1785, interpMode=RCIM_Cubic, leaveTangent=4.1785, tangentMode=RCTM_User, time=0, value=0}
          {arriveTangent=1.7328, interpMode=RCIM_Cubic, leaveTangent=1.7328, tangentMode=RCTM_User, time=0.2042, value=0.6228}
          {arriveTangent=0.0224, interpMode=RCIM_Cubic, leaveTangent=0.0224, tangentMode=RCTM_User, time=1, value=1.5}
         ]}}})
     Passthrough Float=&FloatFromCurve(FloatCurve=DataInterface() {propertyValues {Curve {keys [
          {arriveTangent=4.1785, interpMode=RCIM_Cubic, leaveTangent=4.1785, tangentMode=RCTM_User, time=0, value=0}
          {arriveTangent=1.1179, interpMode=RCIM_Cubic, leaveTangent=1.1179, tangentMode=RCTM_User, time=0.2042, value=0.6228}
          {arriveTangent=0.0224, interpMode=RCIM_Cubic, leaveTangent=0.0224, tangentMode=RCTM_User, time=1, value=1.3}
         ]}}})
    )
   )
   &Drag(
    Drag=&MaskFloatBySpawnGroup(
     Masked Float=0
     Passthrough Float=&FloatFromCurve(
      FloatCurve=DataInterface() {propertyValues {Curve {keys [
          {arriveTangent=2.8162, interpMode=RCIM_Cubic, leaveTangent=2.8162, tangentMode=RCTM_User, time=0, value=0}
          {interpMode=RCIM_Cubic, tangentMode=RCTM_User, time=0.5, value=1}
          {arriveTangent=-2.8774, tangentMode=RCTM_User, time=1, value=0}
         ]}}}
      Scale Curve=5
     )
    )
    Use Linear Drag=true
    Use Rotational Drag=true
   )
   &SolveForcesAndVelocity(Clamp Velocity=false, Limit Acceleration=false)
   &CameraOffset(Camera Offset Amount=&MaskFloatBySpawnGroup(Masked Float=$Particles.Initial.CameraOffset, Passthrough Float=10)) {module=CameraOffset001}
   &DynamicMaterialParameters(
    Index 0 Param 1=&FloatFromCurve(FloatCurve=DataInterface() {propertyValues {Curve {keys [
         {arriveTangent=4.1785, interpMode=RCIM_Cubic, leaveTangent=4.1785, tangentMode=RCTM_User, time=0, value=0}
         {arriveTangent=1.1179, interpMode=RCIM_Cubic, leaveTangent=1.1179, tangentMode=RCTM_User, time=0.2042, value=0.6228}
         {arriveTangent=0.0224, interpMode=RCIM_Cubic, leaveTangent=0.0224, tangentMode=RCTM_User, time=1, value=1}
        ]}}})
    Index 0 Param 3=0.5
    Index 0 Param 4=&FloatFromCurve(
     FloatCurve=DataInterface() {propertyValues {Curve {keys [<at 0.5 1>,<at 1 0>]}}}
     Scale Curve=&FNTimeOfDayScalar(
      Night Scale=2
      Morning Scale=2
      Day Scale=&FN_FrontEndFloatModifier(Gameplay=1, Front End=0.5)
      Evening Scale=3
     )
    )
    Index 2 Param 1=&FloatFromCurve(FloatCurve=DataInterface() {propertyValues {Curve {keys [{interpMode=RCIM_Cubic, tangentMode=RCTM_User, time=0, value=0}, {interpMode=RCIM_Cubic, tangentMode=RCTM_User, time=1, value=1}]}}})
    Index 2 Param 2=&FloatFromCurve(
     FloatCurve=DataInterface() {propertyValues {Curve {keys [
         {arriveTangent=1.9056, interpMode=RCIM_Cubic, leaveTangent=1.9056, tangentMode=RCTM_User, time=0, value=0.2}
         {arriveTangent=0.0259, interpMode=RCIM_Cubic, leaveTangent=0.0259, tangentMode=RCTM_User, time=1, value=1.4}
        ]}}}
     Scale Curve=&RandomRangeFloat(Minimum=0.2, Maximum=0.5)
    )
    Index 2 Param 3=&FloatFromCurve(FloatCurve=DataInterface() {propertyValues {Curve {keys [
         {interpMode=RCIM_Cubic, leaveTangent=2.3064, tangentMode=RCTM_Break, time=0, value=1}
         {interpMode=RCIM_Cubic, leaveTangent=11.1233, tangentMode=RCTM_Break, time=1, value=2}
        ]}}})
    Index 2 Param 4=&FloatFromCurve(FloatCurve=DataInterface() {propertyValues {Curve {keys [
         {arriveTangent=-0.6867, interpMode=RCIM_Cubic, leaveTangent=-0.6867, tangentMode=RCTM_User, time=0, value=1}
         {interpMode=RCIM_Cubic, leaveTangent=-1.2956, tangentMode=RCTM_Break, time=1, value=0.75}
        ]}}})
    Module.Param0WriteEnabled=true
    Module.Param0WriteEnabled2=true
    Module.Param1WriteEnabled=false
    Module.Param1WriteEnabled2=true
    Module.Param2WriteEnabled=true
    Module.Param2WriteEnabled2=true
    Module.Param3WriteEnabled=true
    Module.Param3WriteEnabled2=true
    Write Parameter Index 2=true
   )
   /SetVariables(
    Particles.DMP001:Vector4f=&MakeVector4(X=0, Y=0, Z=0, W=0)
    Particles.DMP002:Vector4f=&MakeVector4(X=0, Y=0, Z=0, W=0)
    Particles.DMP003:Vector4f=&MakeVector4(X=0, Y=0, Z=0, W=0)
    Particles.DMP_0:Vector4f=&MakeVector4(X=0, Y=0, Z=0, W=0)
   ) {isSetParameters=true}
  ]
  rendererClasses [&NiagaraSpriteRendererProperties]
  renderers [
   {rendererClass=&NiagaraSpriteRendererProperties}
   {rendererClass=&NiagaraSpriteRendererProperties, rendererIndex=1}
   {rendererClass=&NiagaraSpriteRendererProperties, rendererIndex=2}
  ]
 }
}
summary {
 systemName=NS_Emote_EggBounce_Suction_Cup_Pop
 emitters [
  {emitterName=SimpleSpriteBurst001, rendererClasses [&NiagaraSpriteRendererProperties]}
  {emitterName=Debris_Sprite001, simTarget=GPUComputeSim, rendererClasses [&NiagaraSpriteRendererProperties]}
 ]
 userVariables [
  User.bIsFrontEnd {type=NiagaraBool}
  User.bIsFrontEndPreview {type=NiagaraBool, defaultValue=true}
 ]
}
